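# Conversation database for the Inferno NPC in the NPC Library, giving the character's greeting and routing players to shared conversations.
# The Inferno entry tag leads to one Speech node with a default line and prioritized contextual lines: a first-ever-conversation introduction and fire-themed quips with empty requirements.
# The speech then branches through link nodes to the shared NPCLibrary Quests and BasicServices entry tags.
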ConversationDatabase /NPCLibrary/NPCs/Inferno/NPCLibrary_Conversation_Inferno
entry_points [{tag=Conversation.Entry.Inferno, destinations [#1]}]
nodes [
 ConversationEntryPointNode_7 {
  id=#1
  class=ConversationEntryPointNode
  outputs [#2]
  properties {entryTag {tagName=Conversation.Entry.Inferno}}
 }
 Speech_C_4 {
  id=#2
  class=Speech_C
  outputs [#3, #4]
  properties {
   nodeName=""
   generalConfig {
    defaultMessage="Bring on the heat!"
    contextualMessages [
     {
      message=I have experience as a henchman.
      contextRequirements [{requirement {class=FortControllerRequirement_IsFirstEverConversationWithNPC}}]
      priority {curve {rowName=Tandem.Speech.Priority.FirstEverInteraction}}
     }
     {
      message="Bring on the heat!"
      requirementMatchPolicy=RequireAny
      contextRequirements []
      priority {curve {curveTable=null, rowName=Tandem.Speech.Priority.OutfitReaction}}
     }
     {
      message=I bring the fire.
      requirementMatchPolicy=RequireAny
      contextRequirements []
      priority {curve {curveTable=null, rowName=Tandem.Speech.Priority.OutfitReaction}}
     }
     {
      message=Everyone comes to Inferno like moths to a flame.
      requirementMatchPolicy=RequireAny
      contextRequirements []
      priority {curve {curveTable=null, rowName=Tandem.Speech.Priority.OutfitReaction}}
     }
     {
      message=Wanna burn through the competition?
      requirementMatchPolicy=RequireAny
      contextRequirements []
      priority {curve {curveTable=null, rowName=Tandem.Speech.Priority.OutfitReaction}}
     }
    ]
   }
  }
 }
 ConversationLinkNode_0 {id=#4, class=ConversationLinkNode, properties {remoteEntryTag {tagName=Conversation.Entry.BasicServices}}}
 ConversationLinkNode_1 {id=#3, class=ConversationLinkNode, properties {remoteEntryTag {tagName=Conversation.Entry.NPCLibrary.Quests}}}
]
